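# Physics asset for the Plague pickaxe skeletal mesh, a cosmetic two-handed melee weapon.
# A single capsule body on the Melee_TwoHanded_Handle bone provides the coarse collision and physics shape for the mesh.
# The asset has no other bodies and no constraints, so it defines no articulated or jointed parts.
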PhysicsAsset /BRCosmetics/Weapons/FORT_Melee/Pickaxe_Plague/Meshes/SK_Pickaxe_Plague_PhysicsAsset
bodies {Melee_TwoHanded_Handle [{shapeType=Capsule, length=66.7752, radius=36.3777, center=<0.5337 -1.0529 34.4933>, rotation=<rot 0.0301 -90 -0.0349>}]}
constraints []
